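# Editor utility widget for a one-off pass that copies static mesh Low-quality minimum LOD to Medium.
# GenerateMeshReport finds selected meshes needing the change via registry tag or loaded mesh, fills an asset collection and prints per-plugin counts.
# RunTool reads that collection, keeps source-control-checkoutable meshes, updates a capped batch, saves them and removes them from the collection.
# A details view exposes settings; the generic CopyQualityLevel function is never called.
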EditorUtilityWidgetBlueprint /Game/NeverCook/EditorUtilities/OneOffs/MinLODCopyTool/EUW_MinLODCopyTool
graphs [
 CopyLowToMedium {nodes [
   CopyLowToMedium!() -> (then, InMesh:Static Mesh Object Reference) {is=title K2Node_FunctionEntry}
   n1 = GetMinLODForQualityLevels!(self:Static Mesh Object Reference=Get InMesh() -> (InMesh:Static Mesh Object Reference)) -> (QualityLevelMinimumLODs:Map of EPerQualityLevels Enums to Integers=Low, Default:int=0)
   > Set l_QualityLevels(l_QualityLevels:Map of EPerQualityLevels Enums to Integers=@n1) -> (Output_Get:Map of EPerQualityLevels Enums to Integers=Low)
   > Set l_default(l_default:int=@n1.Default) -> (Output_Get:int=0)
   > n2 = FilterMeshForUpdate!(execute, self, InMesh:Static Mesh Object Reference=Get InMesh()) -> (then, NeedsUpdate:bool, NewValue:int=0, OutMesh:Static Mesh Object Reference)
   > Branch(Condition=@n2)
   > Add!(
    execute
    TargetMap:Map of EPerQualityLevels Enums to Integers=Get l_QualityLevels() -> (l_QualityLevels:Map of EPerQualityLevels Enums to Integers=Low)
    Key:"EPerQualityLevels Enum (by ref)"=Medium
    Value:"Integer (by ref)"=@n2.NewValue
   ) -> (then) {is=title K2Node_CallFunction}
   > SetMinLODQualityLevelsValidated!(
    execute
    self
    Mesh:Static Mesh Object Reference=Get InMesh()
    QualityLevelMinimumLODs:Map of EPerQualityLevels Enums to Integers=Get l_QualityLevels() -> (l_QualityLevels=Low)
    Default:int=Get l_default() -> (l_default:int=0)
   ) -> (then)
  ]}
 CopyQualityLevel {nodes [
   n1 = Keys!(
    execute
    TargetMap:Map of EPerQualityLevels Enums to Integers=GetMinLODForQualityLevels(self=Get InMesh()) -> (QualityLevelMinimumLODs=Low, Default=0)
   ) -> (then, Keys:Array of EPerQualityLevels Enums)
   > n2 = For Each Loop(Array:Array of EPerQualityLevels Enums=@n1) -> (Array Element:EPerQualityLevels Enum)
   > Branch(Condition="Not Equal (Enum)"!(A:EPerQualityLevels Enum=@n2, B:EPerQualityLevels Enum=Get FromLevel() -> (FromLevel:EPerQualityLevels Enum=Low)) -> (ReturnValue:bool))
   CopyQualityLevel!() -> (then, InMesh:Static Mesh Object Reference, FromLevel:EPerQualityLevels Enum=Low, ToLevel:EPerQualityLevels Enum=Low) {is=title K2Node_FunctionEntry}
   n3 = GetMinLODForQualityLevels(self=Get InMesh()) -> (QualityLevelMinimumLODs=Low, Default=0)
   > Set l_QualityLevels(l_QualityLevels=@n3) -> (Output_Get=Low)
   > Set l_default(l_default=@n3.Default) -> (Output_Get=0)
   > n4 = Sequence!(execute) -> (then_0, then_1, then_2:Exec)
   n5 = Find(
    TargetMap:Map of EPerQualityLevels Enums to Integers=Get l_QualityLevels() -> (l_QualityLevels=Low)
    Key:"EPerQualityLevels Enum (by ref)"=Get ToLevel() -> (ToLevel:EPerQualityLevels Enum=Low)
   ) -> (Value:int)
   > n6 = Branch(Condition=@n5.ReturnValue)
   > Add(TargetMap=Get l_QualityLevels() -> (l_QualityLevels=Low), Key=Get FromLevel() -> (FromLevel=Low), Value=@n5)
   n7 = GetNearestQualityLevel!(
    execute=@n6.else
    self
    QualityLevels:Map of EPerQualityLevels Enums to Integers=Get l_QualityLevels() -> (l_QualityLevels=Low)
    FromLevel:EPerQualityLevels Enum=Get FromLevel() -> (FromLevel=Low)
    ToLevel:EPerQualityLevels Enum=Get ToLevel() -> (ToLevel=Low)
   ) -> (then, ClosestLevel:EPerQualityLevels Enum=Low, Index:int=0)
   > n8 = Branch(Condition="Not Equal (Enum)"(A=@n7, B=Num))
   > Add(TargetMap=Get l_QualityLevels() -> (l_QualityLevels=Low), Key=Get FromLevel() -> (FromLevel=Low), Value=@n7.Index)
   Remove!(
    execute=@n8.else
    TargetMap:Map of EPerQualityLevels Enums to Integers=Get l_QualityLevels() -> (l_QualityLevels=Low)
    Key:"EPerQualityLevels Enum (by ref)"=Get FromLevel() -> (FromLevel=Low)
   ) -> (then, ReturnValue:bool)
   SetMinLODQualityLevelsValidated(
    execute=@n4.then_2
    Mesh=Get InMesh()
    QualityLevelMinimumLODs=Get l_QualityLevels() -> (l_QualityLevels=Low)
    Default=Get l_default() -> (l_default=0)
   )
   "To Integer (Byte)"!(InByte:Byte=0) -> (ReturnValue:int=0)
   "To Integer (Byte)"(InByte=Get ToLevel() -> (ToLevel=Low)) -> (ReturnValue=0)
   ForEach EPerQualityLevels!(execute:Exec, SkipHidden:bool) -> (LoopBody:Exec, EnumValue:EPerQualityLevels Enum, then:Exec) {is=title K2Node_ForEachElementInEnum}
   For Loop(FirstIndex="To Integer (Byte)"(InByte=Get FromLevel() -> (FromLevel=Low)) -> (ReturnValue=0))
   Set l_closestLevel(l_closestLevel:EPerQualityLevels Enum=@n2) -> (Output_Get:EPerQualityLevels Enum=Low)
  ]}
 EventGraph {nodes [
   Event Construct()
   > SetObject(self=Get DetailsView_77() -> (DetailsView_77:Details View Object Reference), NewObject=Self-Reference())
  ]}
 FilterAssetForUpdate {nodes [
   n1 = FilterAssetForUpdate!() -> (then, AssetData:Asset Data Structure) {is=title K2Node_FunctionEntry}
   > n2 = GetTagValue!(execute:Exec, InAssetData=@n1, InTagName=MakeLiteralName(Value=QualityLevelMinLOD)) -> (then:Exec, OutTagValue, ReturnValue)
   > n3 = Branch(Condition=@n2.ReturnValue)
   > Set l_Tag(l_Tag:String=@n2) -> (Output_Get:String)
   > n4 = Sequence()
   > n5 = Split!(execute:Exec, SourceString=Get l_Tag() -> (l_Tag:String), InStr="Low=", SearchCase, SearchDir=FromStart) -> (then:Exec, LeftS, RightS, ReturnValue)
   > n6 = Branch(Condition=@n5.ReturnValue)
   > Set l_low(l_low:int=String To Integer(InString=Left!(SourceString:String=@n5.RightS, Count:int=1) -> (ReturnValue:String))) -> (Output_Get:int=0)
   > n7 = Split(SourceString=Get l_Tag(), InStr="Medium=", SearchDir=FromStart)
   > Branch(Condition=@n7.ReturnValue)
   > n8 = Set l_med(l_med:int=String To Integer(InString=Left(SourceString=@n7.RightS, Count=1))) -> (Output_Get:int=0)
   > Branch(Condition="Equal (Integer)"(A=@n8, B=Get l_low() -> (l_low:int=0)))
   > Return Node!(execute, IncludeForUpdate:bool=false)
   Return Node(execute=@n6.else, IncludeForUpdate=false)
   n9 = Break Asset Data(AssetData=Get AssetData())
   Branch(execute=@n4.then_1, Condition=Contains(SearchIn="To String (Name)"(InName=@n9.PackagePath), Substring=PBW))
   > Return Node(IncludeForUpdate=false)
   Return Node(execute=@n4.then_2, IncludeForUpdate=true)
   Return Node(execute=@n3.else, IncludeForUpdate=false)
  ]}
 FilterAssetsByNeedsConversion {nodes [
   n1 = FilterAssetsByNeedsConversion!() -> (then, AssetDatas:Array of Asset Data Structures) {is=title K2Node_FunctionEntry}
   > n2 = For Each Loop(Array:Array of Asset Data Structures=@n1) -> (Array Element:Asset Data Structure)
   Return Node!(execute=@n2.Completed, ValidAssets:Array of Asset Data Structures=Get l_outAssets() -> (l_outAssets:Array of Asset Data Structures))
   > n3 = FilterAssetForUpdate!(execute, self, AssetData:Asset Data Structure=@n2) -> (then, IncludeForUpdate:bool) {is=title K2Node_CallFunction}
   > Branch(Condition=@n3)
   > Add(TargetArray:Array of Asset Data Structures=Get l_outAssets(), NewItem:"Asset Data Structure (by ref)"=@n2) {is=title K2Node_CallArrayFunction}
  ]}
 FilterAssetsBySourceControlState {nodes [
   n1 = FilterAssetsBySourceControlState!() -> (then, AssetDatas:Array of Asset Data Structures) {is=title K2Node_FunctionEntry}
   > n2 = GetFilesPathsFromData!(execute, self, AssetDatas:Array of Asset Data Structures=@n1) -> (then, Files:Array of Strings, FilteredData:Array of Asset Data Structures)
   > n3 = QueryFileStates!(execute, InFiles:Array of Strings=@n2, bSilent:bool=true, bUseSourceControlStateCache:bool=false) -> (then, ReturnValue:Array of Source Control State Structures)
   > n4 = For Each Loop(Array:Array of Source Control State Structures=@n3) -> (Array Element:Source Control State Structure)
   Return Node(execute=@n4.Completed, ValidAssets=Get l_outAssets())
   n5 = Break Source Control State!(SourceControlState:"Source Control State Structure (by ref)"=@n4) -> (
    bIsValid:bool=~
    bCanCheckOut:bool=~
    bIsCurrent:bool=~
    bIsSourceControlled:bool=~
    bIsConflicted:bool=~
    bIsCheckedOutOther:bool=~
    bIsCheckedOutInOtherBranch:bool=~
    bIsModifiedInOtherBranch:bool=~
    PreviousUser:String
   ) {is=title K2Node_BreakStruct}
   > Branch(Condition=AND Boolean!(
     A=@n5
     B=@n5.bCanCheckOut
     C:bool=@n5.bIsCurrent
     D:bool=@n5.bIsSourceControlled
     E:bool=NOT Boolean(A=@n5.bIsConflicted)
     F:bool=NOT Boolean(A=@n5.bIsCheckedOutOther)
     G:bool=NOT Boolean(A=@n5.bIsCheckedOutInOtherBranch)
     H:bool=NOT Boolean(A=@n5.bIsModifiedInOtherBranch)
    ) -> (ReturnValue))
   n6 = "Get (a ref)"(Array:Array of Asset Data Structures=@n2.FilteredData, Dimension 1=@n4.Array Index) -> (Output:"Asset Data Structure (by ref)")
   > Add(TargetArray=Get l_outAssets(), NewItem=@n6)
   n7 = Break Asset Data(AssetData=@n6)
   PrintString(InString="To String (Text)"(InText=Format Text!(
      Format="LOCTEXT(\"{file}----------------------\\r\\nValid: {valid}\\r\\nCanCheckOut: {canCheckOut}\\r\\nIsCurrent: {ISCurrent}\\r\\nIsSourceControlled: {isSCed}\\r\\nIsConflicted: {isConflicted}\\r\\nIsCheckedOutOther: {isCheckedOut}\\r\\nIsCheckedOutBranch: {isCheckedOutBranch}\\r\\nIsModifiedInOtherBranch: {isModifiedInBranch}\")"
      file:Name=@n7.AssetName
      valid:bool=@n5
      canCheckOut:bool=@n5.bCanCheckOut
      ISCurrent:bool=@n5.bIsCurrent
      isSCed:bool=@n5.bIsSourceControlled
      isConflicted:bool=@n5.bIsConflicted
      isCheckedOut:bool=@n5.bIsCheckedOutOther
      isCheckedOutBranch:bool=@n5.bIsCheckedOutInOtherBranch
      isModifiedInBranch:bool=@n5.bIsModifiedInOtherBranch
     ) -> (Result)))
  ]}
 FilterMeshForUpdate {nodes [
   Add(
    TargetMap=Get l_QualityLevels() -> (l_QualityLevels=Low)
    Key=Medium
    Value=Find(TargetMap=Get l_QualityLevels() -> (l_QualityLevels=Low), Key=Low)
   ) {is=title K2Node_CallFunction}
   > Set l_madeChange(l_madeChange:bool=false) -> (Output_Get:bool)
   FilterMeshForUpdate!() -> (then, InMesh:Static Mesh Object Reference) {is=title K2Node_FunctionEntry}
   > Set l_QualityLevels(l_QualityLevels=GetMinLODForQualityLevels(self=Get InMesh()) -> (QualityLevelMinimumLODs=Low, Default=0)) -> (Output_Get=Low)
   > n1 = Sequence!(execute) -> (then_0, then_1)
   > n2 = Find!(execute:Exec, TargetMap=Get l_QualityLevels() -> (l_QualityLevels=Low), Key=Low) -> (then:Exec, Value, ReturnValue)
   > Set l_low(l_low=@n2) -> (Output_Get=0)
   > n3 = Branch(Condition=@n2.ReturnValue)
   > n4 = Find(TargetMap=Get l_QualityLevels() -> (l_QualityLevels=Low), Key=Medium)
   > Set l_medium(l_medium:int=@n4) -> (Output_Get:int=0)
   > Branch(Condition=@n4.ReturnValue)
   > n5 = Branch(Condition="Not Equal (Integer)"(A=Get l_low() -> (l_low=0), B=Get l_medium() -> (l_medium:int=0)))
   Return Node!(execute=@n5.else, NeedsUpdate:bool=false, NewValue:int=0, OutMesh:Static Mesh Object Reference)
   Return Node(execute=@n3.else, NeedsUpdate=false, NewValue=0)
   Return Node(execute=@n1.then_1, NeedsUpdate=true, NewValue=Get l_low() -> (l_low=0), OutMesh=Get InMesh())
   Find!(TargetMap=Get l_QualityLevels() -> (l_QualityLevels=Low), Key=Low) -> (Value, ReturnValue)
  ]}
 GenerateMeshReport {nodes [
   n1 = For Each Loop(Exec=@n14.then_0, Array:Array of Asset Data Structures=Get l_assets() -> (l_assets:Array of Asset Data Structures)) -> (Array Element:Asset Data Structure)
   n2 = Cast To StaticMesh(Object=GetAsset(InAssetData=@n1))
   > n3 = FilterMeshForUpdate(InMesh=@n2) -> (NewValue=0) {is=title K2Node_CallFunction}
   > n4 = Branch(Condition=@n3)
   Add(
    execute=[@n4.then, @n31.then]
    TargetArray=Get l_Datas() -> (l_Datas:Array of Asset Data Structures)
    NewItem=Get l_data() -> (l_data:Asset Data Structure)
   ) {is=title K2Node_CallArrayFunction}
   n5 = "To String (Name)"(InName=Break Asset Data(AssetData=Get l_data()).PackagePath)
   > n6 = Set l_curPlugin(l_curPlugin:String="Get (a copy)"(Array=ParseIntoArray(SourceString=@n5, Delimiter=/))) -> (Output_Get:String)
   > n7 = Find!(
    execute:Exec
    TargetMap:Map of Strings to Min LODCopy Output Struct Structures=Get l_PluginToDataMap() -> (l_PluginToDataMap:Map of Strings to Min LODCopy Output Struct Structures)
    Key:"String (by ref)"=@n6
   ) -> (then:Exec, Value:Min LODCopy Output Struct Structure, ReturnValue)
   > n8 = Branch(Condition=@n7.ReturnValue)
   > n9 = Set l_curAssets(l_curAssets:Array of Asset Data Structures=Break Min LODCopy Output Struct!(MinLODCopyOutputStruct:"Min LODCopy Output Struct Structure (by ref)"=@n7) -> (AssetDatas_6_#1:Array of Asset Data Structures) {is=title K2Node_BreakStruct}) -> (Output_Get:Array of Asset Data Structures)
   > Add(TargetArray=@n9, NewItem=Get l_data())
   > Add(
    TargetMap:Map of Strings to Min LODCopy Output Struct Structures=Get l_PluginToDataMap()
    Key:"String (by ref)"=Get l_curPlugin() -> (l_curPlugin:String)
    Value:"Min LODCopy Output Struct Structure (by ref)"=Make Min LODCopy Output Struct!(AssetDatas_6_#1:Array of Asset Data Structures=Get l_curAssets() -> (l_curAssets:Array of Asset Data Structures)) -> (MinLODCopyOutputStruct:Min LODCopy Output Struct Structure) {is=title K2Node_MakeStruct}
   ) {is=title K2Node_CallFunction}
   n10 = Get l_data()
   Add(
    execute=@n8.else
    TargetMap=Get l_PluginToDataMap()
    Key=Get l_curPlugin()
    Value=Make Min LODCopy Output Struct(AssetDatas_6_#1=Make Array("[0]":Asset Data Structure=@n10) -> (Array:Array of Asset Data Structures))
   )
   GenerateMeshReport() {is=title K2Node_FunctionEntry}
   > n11 = GetSelectedAssetData()
   > n12 = Set l_assets(l_assets:Array of Asset Data Structures=@n11) -> (Output_Get:Array of Asset Data Structures)
   n13 = MakeTopLevelAssetPath!(PackageName:String=&StaticMesh, AssetName:String) -> (ReturnValue:Top Level Asset Path Structure)
   > RunAssetsThroughFilter!(
    execute
    self:Asset Registry Interface=GetAssetRegistry()
    AssetDataList:Array of Asset Data Structures=@n12
    Filter:"ARFilter Structure (by ref)"=MakeARFilter!(
     PackageNames:Array of Names
     PackagePaths:Array of Names
     SoftObjectPaths:Array of Soft Object Path Structures
     ClassPaths:Array of Top Level Asset Path Structures=Make Array("[0]":Top Level Asset Path Structure=@n13) -> (Array:Array of Top Level Asset Path Structures)
     RecursiveClassPathsExclusionSet:Set of Top Level Asset Path Structures
     ClassNames:Array of Names
     RecursiveClassesExclusionSet:Set of Names
     bRecursivePaths:bool=false
     bRecursiveClasses:bool=false
     bIncludeOnlyOnDiskAssets:bool=false
    ) -> (ReturnValue:ARFilter Structure)
   ) -> (then)
   > n14 = Sequence!(execute) -> (then_0, then_1, then_2:Exec)
   n15 = Branch(execute=@n14.then_1, Condition=Get CheckForP4OnReportGeneration() -> (CheckForP4OnReportGeneration:bool))
   > n16 = GetFilesPathsFromData(AssetDatas=Get l_Datas())
   > n17 = Set l_Datas(l_Datas:Array of Asset Data Structures=@n16.FilteredData) -> (Output_Get:Array of Asset Data Structures)
   n18 = Get CollectionName() -> (CollectionName:Name=None)
   n19 = CollectionExists!(
    execute=[@n15.else, @n17.then]
    self:Asset Tags Subsystem Object Reference=Get AssetTagsSubsystem() -> (ReturnValue:Asset Tags Subsystem Object Reference) {is=title K2Node_GetEngineSubsystem}
    Name:Name=@n18
   ) -> (then, ReturnValue:bool)
   > n20 = Branch(Condition=@n19)
   n21 = Get AssetTagsSubsystem()
   > n22 = EmptyCollection!(execute, self:Asset Tags Subsystem Object Reference=@n21, Name:Name=@n18) -> (then, ReturnValue:bool)
   AddAssetDatasToCollection!(
    execute=[@n22.then, @n23.then]
    self:Asset Tags Subsystem Object Reference=@n21
    Name:Name=@n18
    AssetDatas:Array of Asset Data Structures=Get l_Datas()
   ) -> (then, ReturnValue:bool)
   n23 = CreateCollection!(execute=@n20.else, self:Asset Tags Subsystem Object Reference=@n21, Name:Name=@n18, ShareType:ECollectionScriptingShareType Enum=Local) -> (then, ReturnValue:bool)
   PrintString(execute=@n14.then_2, InString="MinLODReport-------------------\r\n\r\n")
   > PrintString(InString="MinLODReport,Plugin,NumMeshes")
   n24 = Get l_PluginToDataMap()
   > n25 = Values!(execute, TargetMap:Map of Strings to Min LODCopy Output Struct Structures=@n24) -> (then, Values:Array of Min LODCopy Output Struct Structures)
   > n26 = Keys(TargetMap:Map of Strings to Min LODCopy Output Struct Structures=@n24) -> (Keys:Array of Strings)
   > n27 = For Each Loop(Array:Array of Strings=@n26) -> (Array Element:String)
   n28 = Length(TargetArray:Array of Asset Data Structures=Break Min LODCopy Output Struct(MinLODCopyOutputStruct="Get (a copy)"(Array:Array of Min LODCopy Output Struct Structures=@n25, Dimension 1=@n27.Array Index) -> (Output:Min LODCopy Output Struct Structure)))
   > PrintString(InString=Append!(A="MinLODReport,", B=@n27, C:String=",", D:String="To String (Integer)"(InInt=@n28)) -> (ReturnValue))
   n29 = Set l_data(execute=@n1.LoopBody, l_data:Asset Data Structure=@n1) -> (Output_Get:Asset Data Structure)
   > n30 = FilterAssetForUpdate(AssetData=@n29)
   > n31 = Branch(Condition=@n30)
  ]}
 GetFilesPathsFromData {nodes [
   n1 = GetFilesPathsFromData!() -> (then, AssetDatas:Array of Asset Data Structures) {is=title K2Node_FunctionEntry}
   > n2 = For Each Loop(Array:Array of Asset Data Structures=@n1) -> (Array Element:Asset Data Structure)
   Return Node!(
    execute=@n2.Completed
    Files:Array of Strings=Get l_files() -> (l_files:Array of Strings)
    FilteredData:Array of Asset Data Structures=Get l_datas() -> (l_datas:Array of Asset Data Structures)
   )
   > n3 = FilterAssetForUpdate(AssetData=@n2)
   > Branch(Condition=@n3)
   > n4 = ConvertAssetPathToFilePath!(execute, InAssetPath:String="To String (Name)"(InName=Break Asset Data(AssetData=@n2)), InExtension:String=.uasset) -> (then, OutPath:String, ReturnValue:bool)
   > Branch(Condition=@n4.ReturnValue)
   > Add(TargetArray:Array of Strings=Get l_files(), NewItem:"String (by ref)"=ConvertToAbsolutePath!(Filename:String=@n4) -> (ReturnValue:String)) {is=title K2Node_CallArrayFunction}
   > Add(TargetArray:Array of Asset Data Structures=Get l_datas(), NewItem:"Asset Data Structure (by ref)"=@n2)
  ]}
 GetNearestQualityLevel {nodes [
   GetNearestQualityLevel!() -> (then, QualityLevels:Map of EPerQualityLevels Enums to Integers=Low, FromLevel:EPerQualityLevels Enum=Low, ToLevel:EPerQualityLevels Enum=Low) {is=title K2Node_FunctionEntry}
   > n1 = Sequence!(execute) -> (then_0, then_1)
   > Set l_checkLevel(l_checkLevel:int="To Integer (Byte)"(InByte=Get ToLevel() -> (ToLevel=Low)) -> (ReturnValue=0)) -> (Output_Get:int=0)
   > Set l_temp(l_temp:int=Get number of entries in EPerQualityLevels!() -> (ReturnValue:int) {is=title K2Node_GetNumEnumEntries}) -> (Output_Get:int=0)
   > Set l_bestIndex(l_bestIndex:int="To Integer (Byte)"(InByte=Get FromLevel() -> (FromLevel=Low)) -> (ReturnValue=0)) -> (Output_Get:int=0)
   n2 = Get FromLevel() -> (FromLevel=Low)
   > Set l_TopToBottom(l_TopToBottom:bool="integer > integer"(A=Get l_checkLevel() -> (l_checkLevel:int=0), B="To Integer (Byte)"(InByte=@n2) -> (ReturnValue=0))) -> (Output_Get:bool)
   n3 = Get l_temp() -> (l_temp:int=0)
   n4 = "To Integer (Byte)"(InByte=Get FromLevel() -> (FromLevel=Low)) -> (ReturnValue=0)
   n5 = Get l_checkLevel() -> (l_checkLevel=0)
   > While Loop!(execute:Exec, Condition:bool=OR Boolean(A="Not Equal (Integer)"(A=@n5, B=@n4), B="Equal (Integer)"(A=@n3))) -> (LoopBody:Exec, Completed:Exec) {is=title K2Node_MacroInstance}
   > Decrement Int(Value=Get l_temp() -> (l_temp=0))
   n6 = Get l_checkLevel() -> (l_checkLevel=0)
   > n7 = Find(
    TargetMap:Map of EPerQualityLevels Enums to Integers=Get QualityLevels() -> (QualityLevels:Map of EPerQualityLevels Enums to Integers=Low)
    Key:"EPerQualityLevels Enum (by ref)"=Byte to Enum EPerQualityLevels!(Byte:Byte="To Byte (Integer)"!(InInt:int=@n6) -> (ReturnValue:Byte=0)) -> (ReturnValue:EPerQualityLevels Enum) {is=title K2Node_CastByteToEnum}
   ) -> (Value:int)
   > Set l_curIndex(l_curIndex:int=@n7) -> (Output_Get:int=0)
   > n8 = Branch(Condition=@n7.ReturnValue)
   n9 = Get l_curIndex() -> (l_curIndex:int=0)
   n10 = Get l_bestIndex() -> (l_bestIndex:int=0)
   n11 = Get l_curIndex() -> (l_curIndex=0)
   n12 = Get l_bestIndex() -> (l_bestIndex=0)
   > Set l_bestIndex(l_bestIndex=Select(
     Option 0:int="Max (Integer)"(A=@n11, B=@n12)
     Option 1:int="Min (Integer)"(A=@n9, B=@n10) -> (ReturnValue=0)
     Index=Get l_TopToBottom() -> (l_TopToBottom:bool)
    ) -> (ReturnValue:int)) -> (Output_Get=0)
   n13 = Get l_checkLevel() -> (l_checkLevel=0)
   > n14 = Set l_bestLevel(l_bestLevel:EPerQualityLevels Enum=Byte to Enum EPerQualityLevels(Byte="To Byte (Integer)"(InInt=@n13) -> (ReturnValue=0))) -> (Output_Get:EPerQualityLevels Enum=Low)
   n15 = Get l_TopToBottom()
   Set l_checkLevel(execute=[@n8.else, @n14.then], l_checkLevel=int + int(A=Get l_checkLevel() -> (l_checkLevel=0), B=SelectInt(A=-1, B=1, bPickA=@n15))) -> (Output_Get=0)
   > PrintText(InText=Format Text!(
     Format="LOCTEXT(\"---\\r\\nBest:{bestIndex}\\r\\nCheckLevel:{check}\\r\\nTtB:{ttb}\")"
     bestIndex:int=Get l_bestIndex() -> (l_bestIndex=0)
     check:int=Get l_checkLevel() -> (l_checkLevel=0)
     ttb:bool=Get l_TopToBottom()
    ) -> (Result))
   Return Node!(
    execute=@n1.then_1
    ClosestLevel:EPerQualityLevels Enum=Get l_bestLevel() -> (l_bestLevel:EPerQualityLevels Enum=Low)
    Index:int=Get l_bestIndex() -> (l_bestIndex=0)
   )
   "integer > integer"(
    A="To Integer (Byte)"(InByte=Get ToLevel() -> (ToLevel=Low)) -> (ReturnValue=0)
    B="To Integer (Byte)"(InByte=Get FromLevel() -> (FromLevel=Low)) -> (ReturnValue=0)
   )
   Set l_foundLODLevel(l_foundLODLevel:int=0) -> (Output_Get:int=0)
  ]}
 RunTool {nodes [
   n1 = GetSelectedAssetData()
   > Set l_assets(l_assets=@n1)
   RunTool() {is=title K2Node_FunctionEntry}
   > Clear(TargetArray:Array of Static Mesh Object References=Get ChangedMeshes() -> (ChangedMeshes:Array of Static Mesh Object References))
   > n2 = GetAssetsInCollection!(execute, self:Asset Tags Subsystem Object Reference=Get AssetTagsSubsystem(), Name:Name=001_BryceMeshPostAudit) -> (then, ReturnValue:Array of Asset Data Structures)
   > Set l_assets(l_assets=@n2)
   n3 = MakeTopLevelAssetPath(PackageName=&StaticMesh)
   > RunAssetsThroughFilter(
    self=GetAssetRegistry()
    AssetDataList=Get l_assets()
    Filter=MakeARFilter(ClassPaths=Make Array("[0]"=@n3), bRecursivePaths=false, bRecursiveClasses=false, bIncludeOnlyOnDiskAssets=false)
   )
   > n4 = Sequence!(execute) -> (then_0, then_1, then_2:Exec)
   > n5 = FilterAssetsBySourceControlState!(execute, self, AssetDatas:Array of Asset Data Structures=Get l_assets()) -> (then, ValidAssets:Array of Asset Data Structures) {is=title K2Node_CallFunction}
   > Set l_assets(l_assets=@n5)
   n6 = Get l_assets()
   n7 = For Loop(
    execute=@n4.then_1
    LastIndex="Min (Integer)"(A=Last Index!(TargetArray:Array of Asset Data Structures=@n6) -> (ReturnValue=0), B=1000) -> (ReturnValue=0)
   )
   SaveLoadedAssets(execute=@n7.Completed, AssetsToSave=Get ChangedMeshes(), bOnlyIfIsDirty=false)
   n8 = Get AssetTagsSubsystem()
   n9 = Get CollectionName() -> (CollectionName=None)
   > n10 = CollectionExists(self=@n8, Name=@n9)
   > Branch(Condition=@n10)
   > RemoveAssetPtrsFromCollection!(execute, self:Asset Tags Subsystem Object Reference=@n8, Name:Name=@n9, AssetPtrs:Array of Object References=Get ChangedMeshes()) -> (then, ReturnValue:bool)
   n11 = Get l_assets()
   n12 = Cast To StaticMesh(execute=@n7.LoopBody, Object=GetAsset(InAssetData="Get (a ref)"(Array=@n11, Dimension 1=@n7)))
   > CopyLowToMedium!(execute, self, InMesh:Static Mesh Object Reference=@n12) -> (then) {is=title K2Node_CallFunction}
  ]}
 SetMinLODQualityLevelsValidated {nodes [
   n1 = SetMinLODQualityLevelsValidated!() -> (then, Mesh:Static Mesh Object Reference, QualityLevelMinimumLODs:Map of EPerQualityLevels Enums to Integers=Low, Default:int) {is=title K2Node_FunctionEntry}
   > SetMinLODForQualityLevels!(
    execute
    self:Static Mesh Object Reference=@n1
    QualityLevelMinimumLODs:Map of EPerQualityLevels Enums to Integers=@n1.QualityLevelMinimumLODs
    Default:int=@n1.Default
   ) -> (then)
   > Add(TargetArray:Array of Static Mesh Object References=Get ChangedMeshes(), NewItem:"Static Mesh Object Reference (by ref)"=@n1)
  ]}
 TestReport {nodes [
   TestReport() {is=title K2Node_FunctionEntry}
   > n1 = GetSelectedAssetData()
   > n2 = Set l_assets(l_assets=@n1)
   n3 = MakeTopLevelAssetPath(PackageName=&StaticMesh)
   > RunAssetsThroughFilter(
    self=GetAssetRegistry()
    AssetDataList=@n2
    Filter=MakeARFilter(ClassPaths=Make Array("[0]"=@n3), bRecursivePaths=false, bRecursiveClasses=false, bIncludeOnlyOnDiskAssets=false)
   )
   > Sequence()
   > FilterAssetsBySourceControlState(AssetDatas=Get l_assets())
   n4 = For Each Loop(Array=Get l_assets())
   > n5 = Set l_data(l_data=@n4)
   > n6 = FilterAssetForUpdate(AssetData=@n5)
   n7 = Break Asset Data(AssetData=@n5)
   > PrintString(InString=Append!(A="To String (Name)"(InName=@n7.AssetName), B=" == ", C="To String (Boolean)"(InBool=@n6)) -> (ReturnValue))
  ]}
]
variables [
 ChangedMeshes {
  type="array<StaticMesh>"
  default [
   null
   null
   null
   null
   null
   null
   null
   null
   null
   null
   null
   null
   null
   null
   null
   null
   null
   null
   null
   null
   null
   null
   null
   null
   null
   null
   null
   null
   null
   null
   null
   null
   null
   null
   null
   null
   null
   null
   null
   null
   null
   null
   null
   null
   null
   null
   null
   null
   null
   null
   null
   null
   null
   null
   null
   null
   null
   null
   null
   null
   null
   null
   null
   null
   null
   null
   null
   null
   null
   null
   null
   null
   null
   null
   null
   null
   null
   null
   null
   null
   null
   null
   null
   null
   null
   null
   null
   null
   null
   null
   null
   /Ascender/Gameplay/Ascender/Meshes/Ascender
   /Ascender/Gameplay/Ascender/Meshes/Ascender_End_Cap
   /Ascender/Gameplay/Ascender/Meshes/Ascender_Handle
   /Ascender/Gameplay/Ascender/Meshes/Jungle_Ascender_Base
   /Ascender/Gameplay/Ascender/Meshes/Jungle_Ascender_End_Cap
   /Ascender/Gameplay/Ascender/Meshes/Jungle_Ascender_Handle
   /Ascender/Gameplay/Ascender/Meshes/Jungle_Ascender_Top
   null
   null
   null
   null
   null
   null
   null
   null
   null
   null
   null
   null
   null
   null
   null
   null
   null
   null
   null
   null
   null
   null
   null
   null
   null
   null
   null
   null
   null
   null
   null
   null
   null
   null
   null
   null
   null
   null
   null
   null
   null
   null
   null
   null
   null
   null
   null
   null
   null
   null
   null
   null
   null
   null
   /BigBushGrenade/Meshes/Leaf_Grenade
   /BladeHazard/Assets/Gameplay_NitroDromeBlades_A/Meshes/SM_Gameplay_NitroDromeBlade_A
   /BladeHazard/Assets/Gameplay_NitroDromeBlades_A/Meshes/SM_Gameplay_NitroDromeBladeBase_A
   /BlastBerryMapContent/Game/Athena/Aircraft/PartyBus_RT/Meshes/PartyBus_RT_PreFlight
   /BlastBerryMapContent/Game/Athena/Environments/Landscape/Foliage/Athena_SM_WavyGrass_TopologyColor
   /BlastBerryMapContent/Game/Athena/Environments/Landscape/Foliage/SM_Athena_DesertGrass01
   /BlastBerryMapContent/Game/Athena/Environments/Landscape/Meshes/Elevation/S_Cave_Ceiling
   /BlastBerryMapContent/Game/Athena/Environments/Landscape/Meshes/Elevation/S_Cave_Entrance
   /BlastBerryMapContent/Game/Athena/Environments/Landscape/Meshes/Elevation/S_Cave_Inner_Concave
   /BlastBerryMapContent/Game/Athena/Environments/Landscape/Meshes/Elevation/S_Cave_Inner_Convex
   /BlastBerryMapContent/Game/Athena/Environments/Landscape/Meshes/Elevation/S_Cave_Inner_Floor
   /BlastBerryMapContent/Game/Athena/Environments/Landscape/Meshes/Elevation/S_Cave_Inner_Floor_To_Concave
   /BlastBerryMapContent/Game/Athena/Environments/Landscape/Meshes/Elevation/S_Cave_Inner_Middle
   /BlastBerryMapContent/Game/Athena/Environments/Landscape/Meshes/Elevation/S_Cave_Inner_Middle1
   /BlastBerryMapContent/Game/Athena/Environments/Landscape/Meshes/Elevation/S_Cave_Inner_Middle2
   /BlastBerryMapContent/Game/Athena/Environments/Landscape/Meshes/Elevation/S_Cave_Inner_Middle_To_Concave
   /BlastBerryMapContent/Game/Athena/Environments/Landscape/Meshes/Elevation/S_Cave_Outer_Convex
   /BlastBerryMapContent/Game/Athena/Environments/Landscape/Meshes/Elevation/S_Cave_Outer_Floor
   /BlastBerryMapContent/Game/Athena/Environments/Landscape/Meshes/Elevation/S_Cave_Outer_Floor_To_Concave
   /BlastBerryMapContent/Game/Athena/Environments/Landscape/Meshes/Elevation/S_Cave_Outer_Middle2
   /BlastBerryMapContent/Game/Athena/Environments/Landscape/Meshes/Elevation/S_Cave_Outer_Middle_To_Concave
   /BlastBerryMapContent/Game/Athena/Environments/Landscape/Meshes/Elevation/S_Cave_Ramp2_FloorTunnel
   /BlastBerryMapContent/Game/Athena/Environments/Landscape/Meshes/Elevation/S_Cave_Ramp_Cap
   /BlastBerryMapContent/Game/Athena/Environments/Landscape/Meshes/Elevation/S_Cave_Ramp_Floor
   /BlastBerryMapContent/Game/Athena/Environments/Landscape/Meshes/Elevation/S_Cave_Ramp_Ground_Straight
   /BlastBerryMapContent/Game/Athena/Environments/Landscape/Meshes/Elevation/S_Cave_Straight2_Middle
   /BlastBerryMapContent/Game/Athena/Environments/Landscape/Meshes/Elevation/S_Cave_Straight2_Middle2
   /BlastBerryMapContent/Game/Athena/Environments/Landscape/Meshes/Elevation/S_Cave_Straight_Concave
   /BlastBerryMapContent/Game/Athena/Environments/Landscape/Meshes/Elevation/S_Cave_Straight_Convex
   /BlastBerryMapContent/Game/Athena/Environments/Landscape/Meshes/Elevation/S_Cave_Straight_Floor
   /BlastBerryMapContent/Game/Athena/Environments/Landscape/Meshes/Elevation/S_Cave_Straight_Middle2
   /BlastBerryMapContent/Game/Athena/Environments/Landscape/Meshes/Elevation/S_Cave_Straight_Middle_To_Concave
   /BlastBerryMapContent/Game/Athena/Environments/Landscape/Meshes/Elevation/S_Cave_TopInner_To_BottomOuter
   /BlastBerryMapContent/Game/Athena/Environments/Landscape/Meshes/Elevation/S_Cave_TopInner_To_BottomStraight
   /BlastBerryMapContent/Game/Athena/Environments/Landscape/Meshes/Elevation/S_Cave_TopStraight_To_BottomOuter
   /BlastBerryMapContent/Game/Athena/Environments/Landscape/Meshes/Elevation/S_Cave_Tunnel_EndCap
   /BlastBerryMapContent/Game/Athena/Environments/Landscape/Meshes/Elevation/S_Cave_Tunnel_Straight
   /BlastBerryMapContent/Game/Athena/Environments/Landscape/Meshes/Elevation/S_Cave_Tunnel_Straight_x2
   /BlastBerryMapContent/Game/Athena/Environments/Landscape/Meshes/Elevation/S_Cave_Tunnel_TSection
   /BlastBerryMapContent/Game/Athena/Environments/Landscape/Meshes/Elevation/S_Cave_Tunnel_Turn
   /BlastBerryMapContent/Game/Athena/Environments/Landscape/Meshes/Elevation/S_Cliff_1x05
   /BlastBerryMapContent/Game/Athena/Environments/Landscape/Meshes/Elevation/S_Cliff_1x1_To_BottomStraight
   /BlastBerryMapContent/Game/Athena/Environments/Landscape/Meshes/Elevation/S_Cliff_1x5
   /BlastBerryMapContent/Game/Athena/Environments/Landscape/Meshes/Elevation/S_Cliff_Inner2
   /BlastBerryMapContent/Game/Athena/Environments/Landscape/Meshes/Elevation/S_Cliff_Outer2
   /BlastBerryMapContent/Game/Athena/Environments/Landscape/Meshes/Elevation/S_Cliff_Outer3
   /BlastBerryMapContent/Game/Athena/Environments/Landscape/Meshes/Elevation/S_Cliff_SBend
   /BlastBerryMapContent/Game/Athena/Environments/Landscape/Meshes/Elevation/S_CliffShort_Inner
   /BlastBerryMapContent/Game/Athena/Environments/Landscape/Meshes/Elevation/S_CliffShort_Outer
  ]
 }
 CheckForP4OnReportGeneration {type=boolean, default=true}
 CollectionName {type=string, default=001_BryceMeshPostAudit}
 From Level {type=EPerQualityLevels, default=Medium}
 NewVar {type=boolean}
 To Level {type=EPerQualityLevels, default=Low}
]
widgets {
 info {rootWidgetClass=&StackBox, widgetCount=2}
 widgets [
  StackBox_0 {widgetClassPath=&StackBox, widget=StackBox_0}
  DetailsView_77 {
   widgetClassPath=&DetailsView
   parent=StackBox_0
   widget=DetailsView_77
   properties {categoriesToShow [nocategory]}
   slot {class=StackBoxSlot, properties {size {sizeRule=Fill, value=1}}}
  }
 ]
}
